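# Physics asset for the dynamic shirt bones of the Racer Zero medium male player body cosmetic.
# Five two-bone chains (back, back-left, back-right, left, right) each pair a tiny capsule master body with a box panel, around a pelvis capsule.
# Each panel is constrained to its master with moderate swing and tight twist limits.
# Each master joins the pelvis with free swing and twist, so the angle limits set there do not restrict it.
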PhysicsAsset /BRCosmetics/Characters/Player/Male/Medium/Bodies/M_MED_Racer_Zero/Meshes/M_MED_Racer_Zero_Dyn
bodies {
 dyn_shirt_bk_01 [{shapeType=Box, extentX=7.5, extentY=13.0996, extentZ=1.8698, center=<5.4514 0.0004 0>}]
 dyn_shirt_bk_MSTR [{shapeType=Capsule, length=0.505, radius=0.505}]
 dyn_shirt_bk_l_01 [{shapeType=Box, extentX=7.5, extentY=13.0996, extentZ=1.8698, center=<5.4514 0.0004 0>}]
 dyn_shirt_bk_l_MSTR [{shapeType=Capsule, length=0.505, radius=0.505}]
 dyn_shirt_bk_r_01 [{shapeType=Box, extentX=7.5, extentY=13.0996, extentZ=1.8698, center=<5.4514 0.0004 0>}]
 dyn_shirt_bk_r_MSTR [{shapeType=Capsule, length=0.505, radius=0.505}]
 dyn_shirt_l_01 [{shapeType=Box, extentX=7.5, extentY=13.0996, extentZ=1.8698, center=<5.4514 0.0004 0>}]
 dyn_shirt_l_MSTR [{shapeType=Capsule, length=0.505, radius=0.505}]
 dyn_shirt_r_01 [{shapeType=Box, extentX=7.5, extentY=13.0996, extentZ=1.8698, center=<5.4514 0.0004 0>}]
 dyn_shirt_r_MSTR [{shapeType=Capsule, length=0.505, radius=0.505}]
 pelvis [{shapeType=Capsule, length=8.7745, radius=10.8041, center=<-2.4193 -0.4175 0>}]
}
constraints [
 {bone1Name=dyn_shirt_bk_01, bone2Name=dyn_shirt_bk_MSTR, swing1LimitDegrees=30, swing2LimitDegrees=30, twistLimitDegrees=5}
 {bone1Name=dyn_shirt_bk_l_01, bone2Name=dyn_shirt_bk_l_MSTR, swing1LimitDegrees=30, swing2LimitDegrees=30, twistLimitDegrees=5}
 {bone1Name=dyn_shirt_l_01, bone2Name=dyn_shirt_l_MSTR, swing1LimitDegrees=30, swing2LimitDegrees=30, twistLimitDegrees=5}
 {bone1Name=dyn_shirt_bk_r_01, bone2Name=dyn_shirt_bk_r_MSTR, swing1LimitDegrees=30, swing2LimitDegrees=30, twistLimitDegrees=5}
 {bone1Name=dyn_shirt_r_01, bone2Name=dyn_shirt_r_MSTR, swing1LimitDegrees=30, swing2LimitDegrees=30, twistLimitDegrees=5}
 {
  bone1Name=dyn_shirt_bk_MSTR
  bone2Name=pelvis
  swing1LimitDegrees=45
  swing1Motion=Free
  swing2LimitDegrees=45
  swing2Motion=Free
  twistLimitDegrees=45
  twistMotion=Free
 }
 {
  bone1Name=dyn_shirt_bk_l_MSTR
  bone2Name=pelvis
  swing1LimitDegrees=45
  swing1Motion=Free
  swing2LimitDegrees=45
  swing2Motion=Free
  twistLimitDegrees=45
  twistMotion=Free
 }
 {
  bone1Name=dyn_shirt_l_MSTR
  bone2Name=pelvis
  swing1LimitDegrees=45
  swing1Motion=Free
  swing2LimitDegrees=45
  swing2Motion=Free
  twistLimitDegrees=45
  twistMotion=Free
 }
 {
  bone1Name=dyn_shirt_bk_r_MSTR
  bone2Name=pelvis
  swing1LimitDegrees=45
  swing1Motion=Free
  swing2LimitDegrees=45
  swing2Motion=Free
  twistLimitDegrees=45
  twistMotion=Free
 }
 {
  bone1Name=dyn_shirt_r_MSTR
  bone2Name=pelvis
  swing1LimitDegrees=45
  swing1Motion=Free
  swing2LimitDegrees=45
  swing2Motion=Free
  twistLimitDegrees=45
  twistMotion=Free
 }
]
