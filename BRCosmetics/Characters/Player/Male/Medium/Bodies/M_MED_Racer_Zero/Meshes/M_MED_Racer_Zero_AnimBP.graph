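# Post-copy AnimBlueprint for the M_MED_Racer_Zero player body, adding thigh volume correction and secondary physics.
# Copies the attached parent mesh's pose, then offsets front and back thigh volume bones additively, weighted by ThighRotLeft and ThighRotRight.
# Three chained RigidBody nodes follow, each with its own physics asset, simulating relative to pelvis, left calf and right calf.
# The event graph only caches the owning skeletal mesh component on initialization.
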AnimBlueprint /BRCosmetics/Characters/Player/Male/Medium/Bodies/M_MED_Racer_Zero/Meshes/M_MED_Racer_Zero_AnimBP
target_skeleton=&/M_MED_Racer_Zero_Skeleton
graphs [
 AnimGraph {nodes [
   n1 = Get ThighRotLeft() -> (ThighRotLeft:float)
   n2 = "Transform (Modify) Bone - Bone: thigh_volume_bk_l"(
    ComponentPose=Local To Component(LocalPose=Copy Pose From Mesh() {bUseAttachedParent=true})
    Alpha=@n1
    Translation="4.485,-5.330,0.409"
    Rotation="0,0,0"
    Scale="1,1,1"
   ) {
    rotationSpace=BCS_BoneSpace
    scaleMode=BMM_Additive
    scaleSpace=BCS_BoneSpace
    translationMode=BMM_Additive
    translationSpace=BCS_BoneSpace
    alphaBoolBlend {blendOption=Linear}
    boneToModify {boneName=thigh_volume_bk_l}
    translation=<4.485 -5.3303 0.409>
   }
   n3 = "Transform (Modify) Bone - Bone: thigh_volume_fr_l"(ComponentPose=@n2, Alpha=@n1, Translation="-2.144,7.226,-0.000", Rotation="0,0,0", Scale="1,1,1") {
    rotationSpace=BCS_BoneSpace
    scaleMode=BMM_Additive
    scaleSpace=BCS_BoneSpace
    translationMode=BMM_Additive
    translationSpace=BCS_BoneSpace
    alphaBoolBlend {blendOption=Linear}
    boneToModify {boneName=thigh_volume_fr_l}
    translation=<-2.144 7.226>
   }
   n4 = Get ThighRotRight() -> (ThighRotRight:float)
   n5 = "Transform (Modify) Bone - Bone: thigh_volume_bk_r"(ComponentPose=@n3, Alpha=@n4, Translation="4.485,5.33,0.000", Rotation="0,0,0", Scale="1,1,1") {
    rotationSpace=BCS_BoneSpace
    scaleMode=BMM_Additive
    scaleSpace=BCS_BoneSpace
    translationMode=BMM_Additive
    translationSpace=BCS_BoneSpace
    alphaBoolBlend {blendOption=Linear}
    boneToModify {boneName=thigh_volume_bk_r}
    translation=<4.485 5.33>
   }
   n6 = "Transform (Modify) Bone - Bone: thigh_volume_fr_r"(ComponentPose=@n5, Alpha=@n4, Translation="-2.144,-7.226,-0.000", Rotation="0,0,0", Scale="1,1,1") {
    rotationSpace=BCS_BoneSpace
    scaleMode=BMM_Additive
    scaleSpace=BCS_BoneSpace
    translationMode=BMM_Additive
    translationSpace=BCS_BoneSpace
    alphaBoolBlend {blendOption=Linear}
    boneToModify {boneName=thigh_volume_fr_r}
    translation=<-2.144 -7.226>
   }
   n7 = RigidBody(ComponentPose=@n6) {
    bClampLinearTranslationLimitToRefPose=true
    bOverrideWorldGravity=true
    overridePhysicsAsset=&/M_MED_Racer_Zero_Dyn
    simulationSpace=BaseBoneSpace
    alphaBoolBlend {blendOption=Linear}
    baseBoneRef {boneName=pelvis}
    componentLinearAccScale=<0.4 0.4 0.4>
    componentLinearVelScale=<0.4 0.4 0.4>
    overrideWorldGravity {z=-980}
   }
   n8 = RigidBody(ComponentPose=@n7) {
    bClampLinearTranslationLimitToRefPose=true
    bOverrideWorldGravity=true
    overridePhysicsAsset=&/M_MED_Racer_Zero_Dyn_leg_l
    simulationSpace=BaseBoneSpace
    alphaBoolBlend {blendOption=Linear}
    baseBoneRef {boneName=calf_l}
    componentLinearAccScale=<0.4 0.3 0.3>
    componentLinearVelScale=<0.4 0.3 0.3>
    overrideWorldGravity {z=-980}
   }
   n9 = RigidBody(ComponentPose=@n8) {
    bClampLinearTranslationLimitToRefPose=true
    bOverrideWorldGravity=true
    overridePhysicsAsset=&/M_MED_Racer_Zero_Dyn_leg_r
    simulationSpace=BaseBoneSpace
    alphaBoolBlend {blendOption=Linear}
    baseBoneRef {boneName=calf_r}
    componentLinearAccScale=<0.4 0.3 0.3>
    componentLinearVelScale=<0.4 0.3 0.3>
    overrideWorldGravity {z=-980}
   }
   Output Pose(Result=Component To Local(ComponentPose=@n9)) {layerGroup=None}
  ]}
 EventGraph {nodes [
   Event BlueprintInitializeAnimation()
   > Set Skel_Mesh(Skel_Mesh=GetOwnerSkeletalMeshComponent())
  ]}
]
variables [Skel_Mesh {type=SkeletalMeshComponent}]
