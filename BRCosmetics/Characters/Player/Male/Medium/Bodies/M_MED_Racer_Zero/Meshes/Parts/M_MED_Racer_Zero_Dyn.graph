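# Physics asset for the dynamic hair parts of the male medium Racer Zero player body, providing bodies for secondary hair motion.
# Six identical small capsules sit on separate hair bones (A–E and G), alongside a larger capsule on the head.
# Each hair bone is constrained directly to the head in a star layout, with identical wide swing and narrow twist limit angles, rather than chained strand to strand.
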PhysicsAsset /BRCosmetics/Characters/Player/Male/Medium/Bodies/M_MED_Racer_Zero/Meshes/Parts/M_MED_Racer_Zero_Dyn
bodies {
 dyn_hair_A_01 [{shapeType=Capsule, length=2.3856, radius=0.9803, center=<3 0 0>, rotation=<rot 90 0 0>}]
 dyn_hair_B_01 [{shapeType=Capsule, length=2.3856, radius=0.9803, center=<3 0 0>, rotation=<rot 90 0 0>}]
 dyn_hair_C_01 [{shapeType=Capsule, length=2.3856, radius=0.9803, center=<3 0 0>, rotation=<rot 90 0 0>}]
 dyn_hair_D_01 [{shapeType=Capsule, length=2.3856, radius=0.9803, center=<3 0 0>, rotation=<rot 90 0 0>}]
 dyn_hair_E_01 [{shapeType=Capsule, length=2.3856, radius=0.9803, center=<3 0 0>, rotation=<rot 90 0 0>}]
 dyn_hair_G_01 [{shapeType=Capsule, length=2.3856, radius=0.9803, center=<3 0 0>, rotation=<rot 90 0 0>}]
 head [{shapeType=Capsule, length=5, radius=5, rotation=<rot 90 0 0>}]
}
constraints [
 {bone1Name=dyn_hair_A_01, bone2Name=head, swing1LimitDegrees=60, swing2LimitDegrees=60, twistLimitDegrees=10}
 {bone1Name=dyn_hair_B_01, bone2Name=head, swing1LimitDegrees=60, swing2LimitDegrees=60, twistLimitDegrees=10}
 {bone1Name=dyn_hair_C_01, bone2Name=head, swing1LimitDegrees=60, swing2LimitDegrees=60, twistLimitDegrees=10}
 {bone1Name=dyn_hair_D_01, bone2Name=head, swing1LimitDegrees=60, swing2LimitDegrees=60, twistLimitDegrees=10}
 {bone1Name=dyn_hair_E_01, bone2Name=head, swing1LimitDegrees=60, swing2LimitDegrees=60, twistLimitDegrees=10}
 {bone1Name=dyn_hair_G_01, bone2Name=head, swing1LimitDegrees=60, swing2LimitDegrees=60, twistLimitDegrees=10}
]
